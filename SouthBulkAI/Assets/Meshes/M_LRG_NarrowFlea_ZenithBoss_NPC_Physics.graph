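# Physics asset for the NarrowFlea Zenith boss NPC mesh, covering a character's upper body and the ship parts on the same skeleton.
# Each simulated bone has one capsule, with no leg bodies and only a right-side hatch.
# Constraints branch from root into a pelvis-spine chain carrying head and arms, and a ship_main branch carrying seat, hatch and thruster.
# Every joint shares the same uniform swing and twist limits.
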PhysicsAsset /SouthBulkAI/Assets/Meshes/M_LRG_NarrowFlea_ZenithBoss_NPC_Physics
bodies {
 hand_l [{shapeType=Capsule, length=10.3818, radius=8.1661, center=<8.5029 -2.1154 0.1099>, rotation=<rot 0 82.0554 -70.6707>}]
 hand_r [{shapeType=Capsule, length=10.3818, radius=8.1661, center=<-8.5028 2.1155 -0.1099>, rotation=<rot 0 82.0554 -70.6707>}]
 head [{shapeType=Capsule, length=16.3925, radius=14.2039, center=<-0.8806 1.2466 -0.5046>, rotation=<rot 5.1218 -90 1.4514>}]
 lowerarm_l [{shapeType=Capsule, length=13.8958, radius=9.807, center=<14.6745 -0.3023 0.2496>, rotation=<rot 0 -88.0046 -74.4701>}]
 lowerarm_r [{shapeType=Capsule, length=13.8958, radius=9.807, center=<-14.6745 0.3023 -0.2495>, rotation=<rot 0 -88.0046 -74.4701>}]
 pelvis [{shapeType=Capsule, length=33.6544, radius=29.684, center=<1.4852 10.9824 0>, rotation=<rot 90 -90 0>}]
 root [{shapeType=Capsule, length=0.505, radius=0.505, rotation=<rot 0 -90 0>}]
 ship_hatch_r [{shapeType=Capsule, length=29.4332, radius=15.503, center=<-1.3111 -2.5591 -2.3148>, rotation=<rot 0 -93.117 -89.7925>}]
 ship_main [{shapeType=Capsule, length=41.6585, radius=46.7676, center=<-1.5563 -0.0047 -2.1621>, rotation=<rot 0 179.9893 -89.9913>}]
 ship_seat [{shapeType=Capsule, length=27.7566, radius=24.5779, center=<-7.5404 -0.0661 4.0082>, rotation=<rot -3.4292 -90 -22.8624>}]
 ship_thruster [{shapeType=Capsule, length=13.3981, radius=13.1904, center=<11.19 0 -0.2244>, rotation=<rot 0 -90 4.1015>}]
 spine_02 [{shapeType=Capsule, length=26.4065, radius=6.6902, center=<4.7538 -14.5858 0>, rotation=<rot 0 -90 0>}]
 spine_03 [{shapeType=Capsule, length=29.0927, radius=24.4272, center=<6.0911 1.7189 -0.1662>, rotation=<rot -6.3215 -90 1.3025>}]
 spine_04 [{shapeType=Capsule, length=27.0438, radius=25.3992, center=<1.9468 1.6568 -0.0337>, rotation=<rot 0 -179.5948 -164.7512>}]
 upperarm_l [{shapeType=Capsule, length=21.6918, radius=7.5241, center=<13.9484 -0.5707 0.7608>, rotation=<rot 0 -87.1212 -79.0192>}]
 upperarm_r [{shapeType=Capsule, length=21.6918, radius=7.5241, center=<-13.9485 0.5707 -0.7607>, rotation=<rot 0 -87.1212 -79.0192>}]
}
constraints [
 {bone1Name=pelvis, bone2Name=root, swing1LimitDegrees=45, swing2LimitDegrees=45, twistLimitDegrees=45}
 {bone1Name=spine_02, bone2Name=pelvis, swing1LimitDegrees=45, swing2LimitDegrees=45, twistLimitDegrees=45}
 {bone1Name=spine_03, bone2Name=spine_02, swing1LimitDegrees=45, swing2LimitDegrees=45, twistLimitDegrees=45}
 {bone1Name=spine_04, bone2Name=spine_03, swing1LimitDegrees=45, swing2LimitDegrees=45, twistLimitDegrees=45}
 {bone1Name=upperarm_l, bone2Name=spine_04, swing1LimitDegrees=45, swing2LimitDegrees=45, twistLimitDegrees=45}
 {bone1Name=lowerarm_l, bone2Name=upperarm_l, swing1LimitDegrees=45, swing2LimitDegrees=45, twistLimitDegrees=45}
 {bone1Name=hand_l, bone2Name=lowerarm_l, swing1LimitDegrees=45, swing2LimitDegrees=45, twistLimitDegrees=45}
 {bone1Name=upperarm_r, bone2Name=spine_04, swing1LimitDegrees=45, swing2LimitDegrees=45, twistLimitDegrees=45}
 {bone1Name=lowerarm_r, bone2Name=upperarm_r, swing1LimitDegrees=45, swing2LimitDegrees=45, twistLimitDegrees=45}
 {bone1Name=hand_r, bone2Name=lowerarm_r, swing1LimitDegrees=45, swing2LimitDegrees=45, twistLimitDegrees=45}
 {bone1Name=head, bone2Name=spine_04, swing1LimitDegrees=45, swing2LimitDegrees=45, twistLimitDegrees=45}
 {bone1Name=ship_main, bone2Name=root, swing1LimitDegrees=45, swing2LimitDegrees=45, twistLimitDegrees=45}
 {bone1Name=ship_seat, bone2Name=ship_main, swing1LimitDegrees=45, swing2LimitDegrees=45, twistLimitDegrees=45}
 {bone1Name=ship_hatch_r, bone2Name=ship_main, swing1LimitDegrees=45, swing2LimitDegrees=45, twistLimitDegrees=45}
 {bone1Name=ship_thruster, bone2Name=ship_main, swing1LimitDegrees=45, swing2LimitDegrees=45, twistLimitDegrees=45}
]
