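# Translucent unlit deferred decal material that draws a circular target reticle, by its name the marker for a homing rocket.
# Two centred sphere masks, a soft outer minus a hard inner, form a ring without textures.
# PercentDistanceFromTarget drives both radii, so the ring contracts into a soft disc as it falls to zero.
# The same parameter fades the shape in as it drops and blends ColorA (orange-red) to ColorB (red), boosted for glow.
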Material /SouthBulkAI/Assets/VFX/M_HomingRocketMarker_Reticle
graph {
 properties {
  blendMode=BLEND_Translucent
  materialDomain=MD_DeferredDecal
  pixelDepthOffsetMode=PDOM_Legacy
  shadingModel=MSM_Unlit
  twoSided=true
  usedShadingModels=Unlit
 }
 nodes [
  n1 = TextureCoordinate()
  n2 = Constant() {r=0.5}
  n3 = ScalarParameter() {defaultValue=1, parameterName=PercentDistanceFromTarget, sliderMax=1}
  n4 = LinearInterpolate(Alpha=@n3) {constA=0.2, constB=0.5}
  n5 = Constant() {r=0.2}
  n6 = TextureCoordinate()
  n7 = Constant() {r=0.5}
  n8 = LinearInterpolate(Alpha=@n3) {constB=0.49}
  n9 = Constant() {r=1}
  Multiply(A=Subtract(A=SphereMask(A=@n1, B=@n2, Radius=@n4, Hardness=@n5), B=SphereMask(A=@n6, B=@n7, Radius=@n8, Hardness=@n9)), B=Subtract(B=@n3))
  Multiply(A=LinearInterpolate(
    A=VectorParameter() {parameterName=ColorA, defaultValue=<rgba 1 0.2057 0 0>}.RGB
    B=VectorParameter() {parameterName=ColorB, defaultValue=<rgba 1 0 0 0>}.RGB
    Alpha=@n3
   )) {constB=5}
 ]
}
