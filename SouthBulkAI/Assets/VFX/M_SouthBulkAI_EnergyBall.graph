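# Unlit Niagara material for the SouthBulkAI energy ball effect, usable on sprites, ribbons and mesh particles.
# A sharpened UV V-axis gradient is added to panning cell noise, then multiplied by a view-angle fade.
# The powered, saturated result blends dark, mid and bright color parameters through Lerp_3Color, feeding a shared emissive adjustment function.
# Colors, noise texture, pan speed and fresnel shaping are exposed as parameters.
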Material /SouthBulkAI/Assets/VFX/M_SouthBulkAI_EnergyBall
graph {
 properties {
  bAutomaticallySetUsageInEditor=false
  bUsedWithNiagaraMeshParticles=true
  bUsedWithNiagaraRibbons=true
  bUsedWithNiagaraSprites=true
  shadingModel=MSM_Unlit
  usedShadingModels=Unlit
 }
 nodes [
  n1 = ComponentMask(None=TextureCoordinate()) {g=true}
  n2 = Saturate(None=Power(Base=OneMinus(None=@n1)) {constExponent=12})
  n3 = ScalarParameter() {parameterName=Noise Pan Speed X}
  n4 = ScalarParameter() {defaultValue=-2, parameterName=Noise Pan Speed Y}
  n5 = TextureSampleParameter2D(UVs=Panner(Speed=AppendVector(A=@n3, B=@n4)) {speedY=2}) {parameterName=NoiseTexture, samplerType=SAMPLERTYPE_LinearGrayscale, texture=/Game/Effects/Fort_Effects/Textures/Tile/T_CellNoise_Tile_M}
  n6 = ScalarParameter() {defaultValue=5, parameterName=Fresnel}
  n7 = Power(
   Base=Multiply(
    A=Add(A=@n2, B=@n5.G)
    B=MaterialFunctionCall("FadeSensitivity (optional)"=@n6) {materialFunction=&FadeBasedOnViewAngle}.Result
   )
   Exp=ScalarParameter() {defaultValue=0.5, parameterName=Fresnel Power}
  ) {constExponent=0.5}
  MaterialFunctionCall(Emissive=MaterialFunctionCall(
    A=VectorParameter() {parameterName=Color_Dark, defaultValue=<rgba 0.4358 0.0019 0 1>}.RGB
    B=VectorParameter() {parameterName=Color_MidRange, defaultValue=<rgba 1 0.6395 0.0221 1>}.RGB
    C=VectorParameter() {parameterName=Color_Bright, defaultValue=<rgba 2 1.5744 0.7292 1>}.RGB
    Alpha=Saturate(None=@n7)
   ) {materialFunction=&Lerp_3Color}.Result) {materialFunction=&MF_FX_EmissiveAdjustments}
 ]
}
