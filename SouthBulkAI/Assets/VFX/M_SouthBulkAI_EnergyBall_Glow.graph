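# Unlit, translucent, two-sided glow material for the SouthBulkAI energy ball VFX, enabled for Niagara sprites, meshes and ribbons.
# An orange-red color parameter runs through an emissive adjustment function into the Emissive reroute.
# A Noise reroute multiplies a panning caustic texture by a sine of stretched UVs, while AngleFade sharpens a view-angle fade.
# A product of reroute values uses TwoSidedSign to treat front and back faces differently.
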Material /SouthBulkAI/Assets/VFX/M_SouthBulkAI_EnergyBall_Glow
graph {
 properties {
  bAutomaticallySetUsageInEditor=false
  bCastRayTracedShadows=false
  bUsedWithNiagaraMeshParticles=true
  bUsedWithNiagaraRibbons=true
  bUsedWithNiagaraSprites=true
  blendMode=BLEND_Translucent
  shadingModel=MSM_Unlit
  twoSided=true
  usedShadingModels=Unlit
 }
 nodes [
  n1 = NamedRerouteUsage()
  Multiply(A=NamedRerouteUsage(), B=LinearInterpolate(B=Saturate(None=@n1), Alpha=TwoSidedSign()) {constA=1})
  NamedRerouteDeclaration(None=MaterialFunctionCall(Emissive=VectorParameter() {parameterName=Color_Emissive, defaultValue=<rgba 2 0.0521 0 1>}.RGB) {materialFunction=&MF_FX_EmissiveAdjustments}.Result) {name=Emissive, nodeColor=<rgba 0.8984 0 1 1>}
  n2 = TextureCoordinate() {vTiling=0.1}
  n3 = TextureSampleParameter2D(UVs=Panner(Coordinate=@n2) {speedX=0.1, speedY=0.5}) {
   parameterName=Noise Texture
   samplerType=SAMPLERTYPE_LinearGrayscale
   texture=/Game/Effects/Fort_Effects/Textures/Tile/T_Caustic_Wind_Diagonal
  }
  n4 = Multiply(A=Sine(None=ComponentMask(None=@n2) {r=true}), B=@n3.G)
  NamedRerouteDeclaration(None=Multiply(A=Power(Base=@n4) {constExponent=1}) {constB=3}) {name=Noise, nodeColor=<rgba 1 0.2578 0 1>}
  n5 = MaterialFunctionCall("FadeSensitivity (optional)"=Constant() {r=0.5}) {materialFunction=&FadeBasedOnViewAngle}
  NamedRerouteDeclaration(None=Saturate(None=Power(Base=@n5.Result) {constExponent=12})) {name=AngleFade, nodeColor=<rgba 0 0.0625 1 1>}
  NamedRerouteUsage()
 ]
}
