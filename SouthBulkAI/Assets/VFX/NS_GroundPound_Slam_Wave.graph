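# Niagara system for a ground pound slam shockwave, exposing ring life, target wave size and a bisJuno flag.
# The Minimal emitter bursts five spawn groups, each tagged by group for its mesh renderers, with per-group switches setting scale, color, alpha and material parameters.
# A system-level interpolated float grows three groups and drives a material lerp on the fifth, while the slightly delayed fourth follows its own short scale curve.
# GroundDark has no renderers.
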NiagaraSystem /SouthBulkAI/Assets/VFX/NS_GroundPound_Slam_Wave
dependencies {
 usedDataInterfaces [&NiagaraDataInterfaceCurve]
 usedDynamicInputs [&MakeCustomFloatFromBool]
 usedModules [
  &SystemState
  &InterpolateOverTime
  &self:SystemSpawnScript.SystemScriptSource.SystemScriptGraph.NiagaraNodeAssignment_0.SetVariables_#1
 ]
}
emitters {
 GroundDark {rendererClasses [], renderers []}
 Minimal {
  emitterUpdateScript [
   &EmitterState(Life Cycle Mode=::NewEnumerator1, Loop Behavior=::NewEnumerator1, Loop Duration=3, UseLoopDelay=false)
   &SpawnBurst_Instantaneous(Use Spawn Probability=false, UseLoopCountLimit=false)
   &SpawnBurst_Instantaneous(Spawn Group=1, Use Spawn Probability=false, UseLoopCountLimit=false) {module=SpawnBurst_Instantaneous001}
   &SpawnBurst_Instantaneous(Spawn Group=2, Use Spawn Probability=false, UseLoopCountLimit=false) {module=SpawnBurst_Instantaneous002}
   &SpawnBurst_Instantaneous(Spawn Group=3, Spawn Time=0.03, Use Spawn Probability=false, UseLoopCountLimit=false) {module=SpawnBurst_Instantaneous003}
   &SpawnBurst_Instantaneous(Spawn Group=4, Use Spawn Probability=false, UseLoopCountLimit=false) {module=SpawnBurst_Instantaneous004}
  ]
  particleSpawnScript [
   &InitializeParticle(
    Color Mode=::NewEnumerator0
    Lifetime=&MergeHelper_FloatSwitch(
     0=$System.lifetime
     1=$System.lifetime
     2=$System.lifetime
     3=0.4
     4=$System.lifetime
     Index=&SpawnGroup()
     Num=::NewEnumerator3
    )
    Mesh Scale Mode=::NewEnumerator3
    Mesh Uniform Scale=&MergeHelper_FloatSwitch(0=4, 1=4.2, 2=4, 3=5, 4=51.59, Index=&SpawnGroup(), Num=::NewEnumerator3)
    Position Offset=&MergeHelper_VectorSwitch(0=<0 0 60>, 1=<0 0 30>, 2=<0 0 60>, 3=<0 0 40>, 4=<0 0 0>, Index=&SpawnGroup(), Num=::NewEnumerator3)
    UsePositionOffset=true
    Write Material Random=true
   )
   /SetVariables(Particles.VisibilityTag=&SpawnGroup()) {isSetParameters=true}
  ]
  particleUpdateScript [
   &ParticleState()
   &Timeline(
    Play Rate=&FloatFromCurve(
     FloatCurve=DataInterface() {propertyValues {Curve {keys [<at 0 1>,<at 1 0>]}}}
     Scale Curve=&MergeHelper_FloatSwitch(0=2.8228, 1=2.8228, 2=2.9517, 3=0, 4=0, Index=&SpawnGroup(), Num=::NewEnumerator3)
    )
    Timeline End=10000
   )
   &ScaleMeshSize(Scale Factor=&MergeHelper_VectorSwitch(
     0=&VectorFromFloat(Value=&Divide_Float(A=$System.InterpolateOverTime.InterpolateOverTime_Float.Moving Average, B=4))
     1=&VectorFromFloat(Value=&Divide_Float(A=$System.InterpolateOverTime.InterpolateOverTime_Float.Moving Average, B=4))
     2=&VectorFromFloat(Value=&Divide_Float(A=$System.InterpolateOverTime.InterpolateOverTime_Float.Moving Average, B=4))
     3=&VectorFromFloat(Value=&FloatFromCurve(FloatCurve=DataInterface() {propertyValues {Curve {keys [
           {interpMode=RCIM_Cubic, leaveTangent=2.7103, tangentMode=RCTM_Break, time=0, value=0.5}
           {arriveTangent=1.6921, tangentMode=RCTM_User, time=1, value=3}
          ]}}}))
     4=<1 1 1>
     Index=&SpawnGroup()
     Num=::NewEnumerator3
    )) {module=ScaleMeshSize001}
   &ScaleColor(
    Scale Alpha=&MergeHelper_FloatSwitch(
     0=&FloatFromCurve(
      FloatCurve=DataInterface() {propertyValues {Curve {
         defaultValue=1
         keys [
          {arriveTangent=9.0725, interpMode=RCIM_Cubic, leaveTangent=9.0725, tangentMode=RCTM_User, time=0, value=0}
          {arriveTangent=0.0615, interpMode=RCIM_Cubic, leaveTangent=0.0615, tangentMode=RCTM_User, time=0.1408, value=0.9847}
          {arriveTangent=-0.1263, interpMode=RCIM_Cubic, leaveTangent=-0.1263, tangentMode=RCTM_User, time=0.7, value=1}
          {arriveTangent=-0.0405, interpMode=RCIM_Cubic, leaveTangent=-0.0405, tangentMode=RCTM_User, time=1, value=0}
         ]
        }}}
      Scale Curve=0.2
     )
     1=&FloatFromCurve(
      FloatCurve=DataInterface() {propertyValues {Curve {
         defaultValue=1
         keys [
          {arriveTangent=9.0725, interpMode=RCIM_Cubic, leaveTangent=9.0725, tangentMode=RCTM_User, time=0, value=0}
          {arriveTangent=0.0615, interpMode=RCIM_Cubic, leaveTangent=0.0615, tangentMode=RCTM_User, time=0.1408, value=2}
          {arriveTangent=-3.4129, interpMode=RCIM_Cubic, leaveTangent=-3.4129, tangentMode=RCTM_User, time=0.6, value=0.75}
          {arriveTangent=-0.0405, interpMode=RCIM_Cubic, leaveTangent=-0.0405, tangentMode=RCTM_User, time=1, value=0}
         ]
        }}}
      Scale Curve=0.5
     )
     2=&FloatFromCurve(FloatCurve=DataInterface() {propertyValues {Curve {
         defaultValue=1
         keys [
          {arriveTangent=9.0725, interpMode=RCIM_Cubic, leaveTangent=9.0725, tangentMode=RCTM_User, time=0, value=0}
          {arriveTangent=0.0615, interpMode=RCIM_Cubic, leaveTangent=0.0615, tangentMode=RCTM_User, time=0.1408, value=0.9847}
          {arriveTangent=-0.1263, interpMode=RCIM_Cubic, leaveTangent=-0.1263, tangentMode=RCTM_User, time=0.7, value=1}
          {arriveTangent=-0.0405, interpMode=RCIM_Cubic, leaveTangent=-0.0405, tangentMode=RCTM_User, time=1, value=0}
         ]
        }}})
     3=&FloatFromCurve(FloatCurve=DataInterface() {propertyValues {Curve {
         defaultValue=1
         keys [
          {arriveTangent=0.0968, interpMode=RCIM_Cubic, leaveTangent=0.0968, tangentMode=RCTM_User, time=0, value=0}
          {arriveTangent=0.0615, interpMode=RCIM_Cubic, leaveTangent=0.0615, tangentMode=RCTM_User, time=0.3, value=0.9847}
          {arriveTangent=-0.1263, interpMode=RCIM_Cubic, leaveTangent=-0.1263, tangentMode=RCTM_User, time=0.45, value=1}
          {arriveTangent=-0.0405, interpMode=RCIM_Cubic, leaveTangent=-0.0405, tangentMode=RCTM_User, time=1, value=0}
         ]
        }}})
     4=&FloatFromCurve(FloatCurve=DataInterface() {propertyValues {Curve {
         defaultValue=1
         keys [
          {arriveTangent=-0.1759, interpMode=RCIM_Cubic, leaveTangent=-0.1759, tangentMode=RCTM_User, time=0, value=0}
          {arriveTangent=0.2847, interpMode=RCIM_Cubic, leaveTangent=0.2847, tangentMode=RCTM_User, time=0.2468, value=1.3414}
          {arriveTangent=-4.8169, interpMode=RCIM_Cubic, leaveTangent=-4.8169, tangentMode=RCTM_User, time=0.7466, value=0.475}
          {arriveTangent=-0.0405, interpMode=RCIM_Cubic, leaveTangent=-0.0405, tangentMode=RCTM_User, time=1, value=0}
         ]
        }}})
     Index=&SpawnGroup()
     Num=::NewEnumerator3
    )
    Scale RGB=&MergeHelper_VectorSwitch(
     0=&VectorFromFloat(Value=&FloatFromCurve(FloatCurve=DataInterface() {propertyValues {Curve {keys [{interpMode=RCIM_Cubic, leaveTangent=-8.0748, tangentMode=RCTM_Break, time=0, value=2}, <at 0.3 1>]}}}))
     1=&VectorFromFloat(Value=&FloatFromCurve(FloatCurve=DataInterface() {propertyValues {Curve {keys [{interpMode=RCIM_Cubic, leaveTangent=-2.19, tangentMode=RCTM_Break, time=0, value=3}, <at 0.3 1>]}}}))
     2=&VectorFromFloat(Value=&FloatFromCurve(FloatCurve=DataInterface() {propertyValues {Curve {keys [{arriveTangent=-1.8543, tangentMode=RCTM_User, time=0.4998, value=0.5}, <at 1 0.2>]}}}))
     3=&VectorFromFloat(Value=&FloatFromCurve(FloatCurve=DataInterface() {propertyValues {Curve {keys [{interpMode=RCIM_Cubic, leaveTangent=-2.19, tangentMode=RCTM_Break, time=0, value=5}, <at 0.2 1>]}}}))
     4=<1 1 1>
     Index=&SpawnGroup()
     Num=::NewEnumerator3
    )
    ScaleA=true
    ScaleRGB=true
   )
   &DynamicMaterialParameters(
    Index 0 Param 1=&MergeHelper_FloatSwitch(
     0=1
     1=1
     2=&FloatFromCurve(FloatCurve=DataInterface() {propertyValues {Curve {keys [
          {arriveTangent=-1.4938, interpMode=RCIM_Cubic, tangentMode=RCTM_Break, time=0, value=1}
          {arriveTangent=-0.6374, interpMode=RCIM_Cubic, leaveTangent=-0.6374, tangentMode=RCTM_User, time=0.2985, value=0.4083}
          {arriveTangent=-0.2537, interpMode=RCIM_Cubic, tangentMode=RCTM_Break, time=1, value=0}
         ]}}})
     3=1
     4=&Lerp_Float(
      A=0.6
      B=-0.2896
      Alpha=&Divide_Float(A=$System.InterpolateOverTime.InterpolateOverTime_Float.Moving Average, B=20)
     )
     Index=&SpawnGroup()
     Num=::NewEnumerator3
    )
    Index 0 Param 2=&FloatFromCurve(FloatCurve=DataInterface() {propertyValues {Curve {
        defaultValue=1
        keys [
         {arriveTangent=1.4216, interpMode=RCIM_Cubic, leaveTangent=1.4216, tangentMode=RCTM_User, time=0, value=0}
         {arriveTangent=1.2384, interpMode=RCIM_Cubic, leaveTangent=1.2384, tangentMode=RCTM_User, time=1.1, value=1.2}
        ]
       }}})
    Index 0 Param 4=$Particles.Timeline.Time
    Module.Index0Float0Write=true
    Module.Index0Float1Write=true
    Module.Index0Float2Write=true
    Module.Index0Float3Write=true
   )
   /SetVariables(Particles.Color=&MergeHelper_ColorSwitch(
     0=$Particles.Color
     1=&MultiplyLinearColor(
      Linear Color=$Particles.Color
      Scale Factor=&MakeVector4FromVectorAndFloat("Vector (RGB)"=<1 1 1>, "Float (Alpha)"=1.35)
     )
     2=&MultiplyLinearColor(
      Linear Color=$Particles.Color
      Scale Factor=&MakeVector4FromVectorAndFloat(
       "Vector (RGB)"=&VectorFromFloat(Value=&FloatFromCurve(
         FloatCurve=DataInterface() {propertyValues {Curve {keys [
             {interpMode=RCIM_Cubic, leaveTangent=0.1425, tangentMode=RCTM_Break, time=0, value=1}
             {arriveTangent=-1.8543, tangentMode=RCTM_User, time=0.5, value=0}
             <at 1 0>
            ]}}}
         Scale Curve=0.8006
        ))
       "Float (Alpha)"=1
      )
     )
     3=$Particles.Color
     4=$Particles.Color
     Index=&SpawnGroup()
     Num=::NewEnumerator3
    )) {isSetParameters=true}
  ]
  rendererClasses [&NiagaraMeshRendererProperties]
  renderers [
   {rendererClass=&NiagaraMeshRendererProperties, rendererIndex=0}
   {rendererClass=&NiagaraMeshRendererProperties, rendererIndex=1}
   {rendererClass=&NiagaraMeshRendererProperties, rendererIndex=2}
   {rendererClass=&NiagaraMeshRendererProperties, rendererIndex=3}
   {rendererClass=&NiagaraMeshRendererProperties, rendererIndex=4}
  ]
 }
}
summary {
 systemName=NS_GroundPound_Slam_Wave
 emitters [{emitterName=Minimal, rendererClasses [&NiagaraMeshRendererProperties]}, {emitterName=GroundDark, rendererClasses []}]
 userVariables [
  User.bisJuno {type=NiagaraBool}
  User.RingLife {type=NiagaraFloat, defaultValue=3, description=""}
  User.TargetWaveSize {type=NiagaraFloat, defaultValue=20, description=""}
 ]
}
