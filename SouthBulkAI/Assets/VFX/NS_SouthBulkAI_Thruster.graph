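# Niagara thruster effect for SouthBulkAI, combining active glow, jet cone, top jet and spark emitters, with one-shot corona and ribbon trail emitters disabled.
# System-level set-variable modules accompany the emitters, whose detail appears only for the trail.
# The blue trail spawns by distance traveled and takes lifetime from system velocity and owner LOD distance.
# Owner scale drives its alpha and width, which also grows with LOD distance.
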NiagaraSystem /SouthBulkAI/Assets/VFX/NS_SouthBulkAI_Thruster
dependencies {
 usedDataInterfaces [&NiagaraDataInterfaceCurve, &NiagaraDataInterfaceColorCurve]
 usedDynamicInputs [
  &RemapRange
  &VectorLength
  &SetBoolByFloatComparison
  &Vector2DFromFloat
  &VectorFromFloat
  &MakeLinearColor
 ]
 usedModules [
  &SystemState
  &self:SystemSpawnScript.SystemScriptSource.SystemScriptGraph.NiagaraNodeAssignment_1.SetVariables_#1
  &self:SystemSpawnScript.SystemScriptSource.SystemScriptGraph.NiagaraNodeAssignment_3.SetVariables_#2
  &self:SystemSpawnScript.SystemScriptSource.SystemScriptGraph.NiagaraNodeAssignment_0.SetVariables_#3
 ]
}
emitters {
 Corona_OneShot {bEnabled=false, rendererClasses [], renderers []}
 Glow_LW {rendererClasses [], renderers []}
 Mesh_JetCone001 {rendererClasses [], renderers []}
 Sparks001 {rendererClasses [], renderers []}
 TopJet001 {rendererClasses [], renderers []}
 Trail {
  bEnabled=false
  emitterUpdateScript [
   &EmitterState()
   &SpawnPerUnit(Module.Use Max Movement Threshold=false, Module.Use Movement Tolerance=false, Spawn Spacing=40, Use Spawn Probability=false)
  ]
  particleSpawnScript [&InitializeParticle(
    Color Mode=::NewEnumerator0
    Lifetime=&Multiply_Float(
     A=$System.Velocity
     B=&FloatFromCurve(FloatCurve=DataInterface() {propertyValues {Curve {keys [<at 1000 0.2>,<at 120000 1.5>]}}}, CurveIndex=$Engine.Owner.LODDistance)
    )
    Ribbon Width=200
    Ribbon Width Mode=::NewEnumerator1
    UsePositionOffset=false
    Write Material Random=true
   )]
  particleUpdateScript [
   &ParticleState()
   &Color(
    Color=&ColorFromCurve(ColorCurve=DataInterface() {propertyValues {
       AlphaCurve {keys [<at 0 0.06 RCIM_Cubic>,<at 0.4126 0.02>,<at 1 0 RCIM_Cubic>]}
       BlueCurve {keys [<at 0 6 RCIM_Cubic>,<at 0.9827 3>]}
       GreenCurve {keys [<at 0 1.9801 RCIM_Cubic>,<at 0.9827 0.9326>]}
       RedCurve {keys [<at 0 0 RCIM_Cubic>,<at 0.9827 0>]}
      }})
    Scale Alpha=&MakeFloatFromVector(Vector=$Engine.Owner.Scale, Channel=::NewEnumerator0)
   )
   &KillParticles(Kill Particles=$System.Bool_IsMoving)
   &SolveForcesAndVelocity(Clamp Velocity=false, Limit Acceleration=false)
   &ScaleRibbonWidth(Ribbon Width Scale=&FloatFromCurve(
     FloatCurve=DataInterface() {propertyValues {Curve {keys [
         {arriveTangent=12.7028, interpMode=RCIM_Cubic, leaveTangent=12.7028, tangentMode=RCTM_User, time=0, value=0.5}
         {arriveTangent=-0.1, interpMode=RCIM_Cubic, leaveTangent=-0.1, time=0.1816, value=1.2277}
         <at 1 0.4>
        ]}}}
     Scale Curve=&FloatFromCurve(
      FloatCurve=DataInterface() {propertyValues {Curve {keys [<at 30000 1>,<at 200000 2>]}}}
      CurveIndex=$Engine.Owner.LODDistance
      Scale Curve=&MakeFloatFromVector(Vector=$Engine.Owner.Scale, Channel=::NewEnumerator0)
     )
    ))
  ]
  rendererClasses [&NiagaraRibbonRendererProperties]
  renderers [{rendererClass=&NiagaraRibbonRendererProperties}]
 }
}
summary {
 systemName=NS_SouthBulkAI_Thruster
 emitters [
  {emitterName=Sparks001, rendererClasses []}
  {emitterName=Mesh_JetCone001, rendererClasses []}
  {emitterName=TopJet001, rendererClasses []}
  {bEnabled=false, emitterName=Trail, rendererClasses [&NiagaraRibbonRendererProperties]}
  {emitterName=Glow_LW, rendererClasses []}
  {bEnabled=false, emitterName=Corona_OneShot, rendererClasses []}
 ]
 userVariables [
  User.RibbonIDOffset {type=NiagaraInt32}
  User.SkeletalMesh {
   type=NiagaraDataInterfaceSkeletalMesh
   defaultValue {
    struct=NiagaraExt_VariableValue_DataInterface
    value {dataInterface=&self:NiagaraDataInterfaceSkeletalMesh_0}
   }
  }
 ]
}
